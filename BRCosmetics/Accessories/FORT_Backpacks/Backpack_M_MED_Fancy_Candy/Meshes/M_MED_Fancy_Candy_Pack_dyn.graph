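# Physics asset for the Fancy Candy backpack cosmetic, providing jointed capsule bodies for its dangling chain and V-Bucks bones.
# The pack body uses two large capsules as the root of the constraints and as collision volume.
# Three chain bodies link in series from the pack, while four V-Bucks bodies each attach directly to it.
# Constraints set per-joint swing limits with very tight twist, limiting how far pieces swing and roll.
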PhysicsAsset /BRCosmetics/Accessories/FORT_Backpacks/Backpack_M_MED_Fancy_Candy/Meshes/M_MED_Fancy_Candy_Pack_dyn
bodies {
 dyn_chains_01 [{shapeType=Capsule, length=1.5, radius=2.5, center=<1 0 0>, rotation=<rot 90 0 0>}]
 dyn_chains_02 [{shapeType=Capsule, length=3.25, radius=2, center=<1.5 0 0>, rotation=<rot 90 0 0>}]
 dyn_chains_03 [{shapeType=Capsule, length=3, radius=2.5, center=<2.5 0 0>, rotation=<rot 90 0 0>}]
 dyn_vbucks_01 [{shapeType=Capsule, length=3.0576, radius=1.6237, center=<2 0 0>, rotation=<rot 0 -167.7139 -83.5758>}]
 dyn_vbucks_02 [{shapeType=Capsule, length=5.1214, radius=2.4719, center=<4.0347 0 -0.193>, rotation=<rot 0 0 -90>}]
 dyn_vbucks_03 [{shapeType=Capsule, length=3.7443, radius=2.1496, center=<2.5 0 0>, rotation=<rot 0 5.8331 -87.8338>}]
 dyn_vbucks_04 [{shapeType=Capsule, length=3.2463, radius=1.6929, center=<1 -0.0994 0>, rotation=<rot 0 -10.9753 -88.4942>}]
 pack [
  {shapeType=Capsule, length=24.3787, radius=13.6445, center=<-0.5123 -0.9691 0.4023>, rotation=<rot 1.1506 -96.5391 -80.062>}
  {shapeType=Capsule, length=24.3787, radius=8.3945, center=<4.4149 -8.9853 -5.8068>, rotation=<rot -1.5446 -90.6209 -128.3754>}
 ]
}
constraints [
 {bone1Name=dyn_chains_01, bone2Name=pack, swing1LimitDegrees=20, swing2LimitDegrees=60, twistLimitDegrees=2}
 {bone1Name=dyn_chains_02, bone2Name=dyn_chains_01, swing1LimitDegrees=20, swing2LimitDegrees=60, twistLimitDegrees=2}
 {bone1Name=dyn_chains_03, bone2Name=dyn_chains_02, swing1LimitDegrees=20, swing2LimitDegrees=60, twistLimitDegrees=2}
 {bone1Name=dyn_vbucks_01, bone2Name=pack, swing1LimitDegrees=30, swing2LimitDegrees=30, twistLimitDegrees=2}
 {bone1Name=dyn_vbucks_02, bone2Name=pack, swing1LimitDegrees=5, swing2LimitDegrees=30, twistLimitDegrees=2}
 {bone1Name=dyn_vbucks_03, bone2Name=pack, swing1LimitDegrees=20, swing2LimitDegrees=45, twistLimitDegrees=2}
 {bone1Name=dyn_vbucks_04, bone2Name=pack, swing1LimitDegrees=15, swing2LimitDegrees=30, twistLimitDegrees=2}
]
